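; Server-authoritative controller component that triggers PoppySoap (MK) gameplay voice-over by applying GameplayEffects to the player's pawn.
; It binds the pawn's damage events, accumulating recent PoppySoap weapon damage per enemy in a map that a looping timer prunes.
; Kills trigger an announcer laugh, or Flawless Victory when recent damage clears a hotfixable threshold and the player has avoided recent damage.
; AnnounceItemAddToInventory applies an announcement effect once per weapon tag.
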(event EventOnPawnSpawned
  (params (Pawn :type Pawn%20Object%20Reference))
  ;[0 0.15 0.02058 0.5] OnPawnSpawned: Bind to know when the pawn does damage or takes damage.
  (|CallParentFunction Pawn)
  (Networking|SwitchHasAuthority
    (:Authority
      (bind _asfort_player_pawn (Utilities|Casting|CastToFortPlayerPawn Pawn))
      (bind _output_get (Variables|Transient|SetPlayerPawn _asfort_player_pawn))
      (EventDispatchers|BindEventtoOnDidDamage _output_get (EventDispatchers|CreateEvent :!event HandleDidDamage))
      (EventDispatchers|BindEventtoOnDamaged _output_get (EventDispatchers|CreateEvent :!event HandleTakeDamage)))))

(event EventOnPawnDied
  (params (KilledPawn :type Fort%20Player%20Pawn%20Object%20Reference) (Damage :type Float%20%28single-precision%29) (InTags :type Gameplay%20Tag%20Container%20Structure%20%28by%20ref%29) (EffectContext :type Gameplay%20Effect%20Context%20Handle%20Structure%20%28by%20ref%29) (EventInstigator :type Controller%20Object%20Reference) (DamageCauser :type Actor%20Object%20Reference) (DBNOFinisher :type Controller%20Object%20Reference))
  ;[0.1417 0.01151 0.15 0.5] OnPawnDied: When the pawn dies, unbind delegates.
  (|CallParentFunction KilledPawn Damage InTags EffectContext EventInstigator DamageCauser DBNOFinisher)
  (Networking|SwitchHasAuthority
    (:Authority
      (Utilities|IsValid KilledPawn
        (:Is%20Valid
          (EventDispatchers|UnbindEventfromOnDidDamage KilledPawn (EventDispatchers|CreateEvent :!event HandleDidDamage))
          (EventDispatchers|UnbindEventfromOnDamaged KilledPawn (EventDispatchers|CreateEvent :!event HandleTakeDamage)))))))

(event EventEndPlay
  (params (EndPlayReason :type EEndPlayReason%20Enum :default "Destroyed"))
  ;[0.15 0.07601 0 0.5] EndPlay: Unbind delegates.
  (Utilities|FlowControl|Sequence
    (:then_0
      (Networking|SwitchHasAuthority
        (:Authority
          (bind _playerpawn (Variables|Transient|GetPlayerPawn)
            (:then
              (EventDispatchers|UnbindEventfromOnDidDamage _playerpawn (EventDispatchers|CreateEvent :!event HandleDidDamage))
              (EventDispatchers|UnbindEventfromOnDamaged _playerpawn (EventDispatchers|CreateEvent :!event HandleTakeDamage)))
            (:else)))))
    (:then_1
      (|CallParentFunction EndPlayReason))))

(fn AnnounceItemAddToInventory
  (params (WeaponGameplayTag :type Gameplay%20Tag%20Structure) (InventoryAddAnnouncementGE :type Gameplay%20Effect%20Class%20Reference))
  ;[0.0279 0.07468 0.15 0.5] AnnounceItemAddToInventory: Invoked when the player picks up an MK item for the first time in a match.
  ; Called from GA_PoppySoap_Inventory.
  (Utilities|SwitchHasAuthority
    (:Authority
      (bind _returnvalue (Utilities|Array|AddUnique (Variables|Transient|GetInventoryAddAnnouncementGameplayTags) WeaponGameplayTag))
      (if (>= _returnvalue 0)
        (CallFunction|ApplyGameplayEffecttoPlayer InventoryAddAnnouncementGE)))))

(fn ApplyGameplayEffectToPlayer
  (params (GameplayEffect :type Gameplay%20Effect%20Class%20Reference))
  ;[0.03226 0.08309 0.15 0.5] ApplyGameplayEffectToPlayer: Applies gameplay effect to the owning player pawn.
  (Networking|SwitchHasAuthority
    (:Authority
      (bind _playerpawn (Variables|Transient|GetPlayerPawn)
        (:then
          (Fortnite|ApplyGameplayEffectsToActor _playerpawn (Utilities|Array|MakeArray GameplayEffect) (0.0 0.0 0.0) false 0 (Ability|GetAbilitySystemComponent _playerpawn) "()"))
        (:else)))))

(fn HandleDidDamage
  (params (DamagedActor :type Actor%20Object%20Reference) (Damage :type Float%20%28single-precision%29) (InstigatedBy :type Controller%20Object%20Reference) (DamageCauser :type Actor%20Object%20Reference) (HitLocation :type Vector) (FHitComponent :type Primitive%20Component%20Object%20Reference) (BoneName :type Name) (Momentum :type Vector))
  (bind _recentdamagedidtrackinginfomap (Variables|Transient|GetRecentDamageDidTrackingInfoMap))
  (bind (_value _returnvalue) (Utilities|Map|Find _recentdamagedidtrackinginfomap (LocalVariables|GetDamagedPlayerPawn)))
  (bind (_totaldamage_5_b359bf7a4ea3fa20a56d0684d0a86a0e _servertimeexpire_4_df1c8ea0450a448b5a0c67abfe7ff53c) (Utilities|Struct|BreakPoppySoapGameplayVORecentDamageTrackingStruct _value))
  (bind _damagedplayerpawn (LocalVariables|GetDamagedPlayerPawn))
  ;[0.1082 0.01903 0.15 0.5] HandleDidDamage: The player did damage; check to see if conditions are satisfied to play necessary gameplay VO.
  (Networking|SwitchHasAuthority
    (:Authority
      (bind _asfort_player_pawn (Utilities|Casting|CastToFortPlayerPawn DamagedActor))
      (LocalVariables|SetDamagedPlayerPawn _asfort_player_pawn)
      (if (Math|Boolean|ANDBoolean (Utilities|Enum|NotEqual(Enum) (Team|GetTeamAffiliation _asfort_player_pawn (Variables|Transient|GetPlayerPawn)) "Friendly") (CallFunction|IsPoppySoapWeapon DamageCauser) (> Damage 0))
        (CallFunction|TrackDidDamagetoEnemy _asfort_player_pawn Damage)
        (Utilities|FlowControl|Sequence
          (:then_0
            ;;[0.000919 0.15 0.08831 0.5] Flawless Victory
            (if (Math|Boolean|ANDBoolean (FortPawn|IsDead _damagedplayerpawn) (not (Health|IsDBNO _damagedplayerpawn)) _returnvalue)
              (if (and (> _totaldamage_5_b359bf7a4ea3fa20a56d0684d0a86a0e (ScalableFloat|GetCachedScalableFloatValue (Variables|Hotfix|GetHotfix_EnemyRecentlyDamagedValueThresholdForFlawlessVictory))) (> _servertimeexpire_4_df1c8ea0450a448b5a0c67abfe7ff53c (Fortnite|Sync|Time|GetServerWorldTimeSeconds)))
                (if (> (- (Fortnite|Sync|Time|GetServerWorldTimeSeconds) (Variables|Transient|GetLastServerTimeTakenDamage)) (ScalableFloat|GetCachedScalableFloatValue (Variables|Hotfix|GetHotfix_RecentlyTakenDamageTrackerDuration)))
                  (CallFunction|ApplyGameplayEffecttoPlayer (Variables|GameplayVOEffects|GetFlawlessVictoryGameplayEffect))))))
          (:then_1
            ;;[0.008241 0.04434 0.15 0.5] Announcer Laugh.
            (if (FortPawn|IsDead (LocalVariables|GetDamagedPlayerPawn))
              (CallFunction|ApplyGameplayEffecttoPlayer (Variables|GameplayVOEffects|GetAnnouncerLaughGameplayEffect)))))))))

(fn HandleTakeDamage
  (params (DamagedActor :type Actor%20Object%20Reference) (Damage :type Float%20%28single-precision%29) (InstigatedBy :type Controller%20Object%20Reference) (DamageCauser :type Actor%20Object%20Reference) (HitLocation :type Vector) (FHitComponent :type Primitive%20Component%20Object%20Reference) (BoneName :type Name) (Momentum :type Vector))
  ;[0.06899 0.05054 0.15 0.5] HandleTakeDamage: When the owning pawn takes damage, cache the server time.
  (Utilities|SwitchHasAuthority
    (:Authority
      (if (and (> Damage 0) (!= InstigatedBy (Components|GetOwner)))
        (Variables|Transient|SetLastServerTimeTakenDamage (Fortnite|Sync|Time|GetServerWorldTimeSeconds))))))

(fn IsPoppySoapWeapon
  (params (WeaponActor :type Actor%20Object%20Reference))
  ;[0.06716 0.06947 0.15 0.5] Uses gameplay tag to see if the weapon actor is an MK weapon.
  (Utilities|FlowControl|Sequence
    (:then_0
      (bind _asfort_weapon (Utilities|Casting|CastToFortWeapon WeaponActor))
      (bind _returnvalue (FortWeapon|GetInventoryItem _asfort_weapon))
      (Utilities|IsValid _returnvalue
        (:Is%20Valid
          (bind _returnvalue_2 (GameplayTags|HasMatchingGameplayTag(Message) _returnvalue "(TagName=\"Item.Weapon.Melee.PoppySoap\")"))
          (return _returnvalue_2))))
    (:then_1
      (return false))))

(fn RecentDamageDidExpirationCheck
  (bind _recentdamagedidtrackinginfomap (Variables|Transient|GetRecentDamageDidTrackingInfoMap))
  (bind _recentdamagedidtrackinginfomap_2 (Variables|Transient|GetRecentDamageDidTrackingInfoMap))
  ;[0.05644 0.1312 0.15 0.5] RecentDamageDidExpirationCheck: Go through the recent damage did tracking map and remove entries that are too old.
  (Networking|SwitchHasAuthority
    (:Authority
      (Utilities|Map|Remove _recentdamagedidtrackinginfomap_2) ; Remove null, player pawns may have expired.
      (if (> (Utilities|Map|Length _recentdamagedidtrackinginfomap_2) 0)
        (bind _returnvalue (Fortnite|Sync|Time|GetServerWorldTimeSeconds))
        (bind _keys (Utilities|Map|Keys _recentdamagedidtrackinginfomap))
        (for _array_element _keys
          (if (.ReturnValue (Utilities|Map|Find _recentdamagedidtrackinginfomap _array_element))
            (if (< (.ServerTimeExpire_4_DF1C8EA0450A448B5A0C67ABFE7FF53C (Utilities|Struct|BreakPoppySoapGameplayVORecentDamageTrackingStruct (Utilities|Map|Find _recentdamagedidtrackinginfomap _array_element))) _returnvalue)
              (Utilities|Map|Remove _recentdamagedidtrackinginfomap _array_element))))
        (if (<= (Utilities|Map|Length _recentdamagedidtrackinginfomap) 0)
          _)
        (else
          _))
      (CallFunction|SetRecentDamageDidExpirationCheckTimerEnable false) ; No need to timer tick anymore when there are no entries in the map.
    )))

(fn SetRecentDamageDidExpirationCheckTimerEnable
  (params (bEnable :type Boolean))
  ;[0.03434 0.06022 0.15 0.5] SetRecentDamageDidExpirationCheckTimerHandle: Enables or disables the timer that is responsible for looping through all the recent damage done to see if any of the damage done is not recent enough and should be removed.
  (Utilities|SwitchHasAuthority
    (:Authority
      (if bEnable
        (if (not (Utilities|Time|IsTimerActivebyHandle (Variables|Transient|GetRecentDamageDidTrackerTimerHandle)))
          (bind _returnvalue (Utilities|Time|SetTimerbyEvent (EventDispatchers|CreateEvent :!event RecentDamageDidExpirationCheck) (ScalableFloat|GetCachedScalableFloatValue (Variables|Hotfix|GetHotfix_EnemyRecentlyDamagedTrackerCheckInterval)) true))
          (Variables|Transient|SetRecentDamageDidTrackerTimerHandle _returnvalue))
        (else
          (Utilities|Time|ClearandInvalidateTimerbyHandle (Variables|Transient|GetRecentDamageDidTrackerTimerHandle)))))))

(fn TrackDidDamageToEnemy
  (params (DamagedPlayerPawn :type Fort%20Player%20Pawn%20Object%20Reference) (Damage :type Float%20%28double-precision%29))
  (bind (_totaldamage_5_b359bf7a4ea3fa20a56d0684d0a86a0e _servertimeexpire_4_df1c8ea0450a448b5a0c67abfe7ff53c) (Utilities|Struct|BreakPoppySoapGameplayVORecentDamageTrackingStruct (Utilities|Map|Find (Variables|Transient|GetRecentDamageDidTrackingInfoMap) DamagedPlayerPawn)))
  (bind (_value _returnvalue) (Utilities|Map|Find (Variables|Transient|GetRecentDamageDidTrackingInfoMap) DamagedPlayerPawn))
  (bind _recentdamagedidtrackinginfomap (Variables|Transient|GetRecentDamageDidTrackingInfoMap))
  ;[0.01556 0.0278 0.15 0.5] TrackDidDamageToEnemy: When the owning player damages an enemy, track the total recent damage done in the map.
  (Networking|SwitchHasAuthority
    (:Authority
      (Utilities|IsValid DamagedPlayerPawn
        (:Is%20Valid
          (if (> Damage 0)
            (if _returnvalue
              (Utilities|Map|Add _recentdamagedidtrackinginfomap (Variables|GetDamagedPlayerPawn) (Utilities|Struct|MakePoppySoapGameplayVORecentDamageTrackingStruct (+ _totaldamage_5_b359bf7a4ea3fa20a56d0684d0a86a0e (Variables|GetDamage)) _servertimeexpire_4_df1c8ea0450a448b5a0c67abfe7ff53c))
              (else
                (Utilities|Map|Add _recentdamagedidtrackinginfomap (Variables|GetDamagedPlayerPawn) (Utilities|Struct|MakePoppySoapGameplayVORecentDamageTrackingStruct (Variables|GetDamage) (+ (Fortnite|Sync|Time|GetServerWorldTimeSeconds) (ScalableFloat|GetCachedScalableFloatValue (Variables|Hotfix|GetHotfix_EnemyRecentlyDamagedTrackerDuration)))))))
            (CallFunction|SetRecentDamageDidExpirationCheckTimerEnable true)))))))
